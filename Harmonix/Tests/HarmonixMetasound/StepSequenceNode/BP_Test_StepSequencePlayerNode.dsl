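; Functional test for the Harmonix MetaSound step sequence player node, exercising it on an audio component while a test widget is on screen.
; Start creates a generator handle and has the widget watch outputs and note on/off, failing if watching fails.
; Chained timers and delays then vary step size, speed and additional octaves at runtime.
; The final step reports success without checking output values; test completion removes the widget and stops audio.
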(event EventPrepareTest
  (bind _returnvalue (UserInterface|CreateWidget "/Harmonix/Tests/HarmonixMetasound/StepSequenceNode/WBP_Test_StepSequence.WBP_Test_StepSequence_C"))
  (bind _output_get (Variables|Default|SetStepSequenceWidget _returnvalue))
  (UserInterface|Viewport|AddtoViewport _output_get 0)
  (EventDispatchers|AssignOnTestFinished (delegate Custom|OnTestFinished_Event)))

(event Custom|OnTestFinished_Event
  (Widget|RemovefromParent (stale EventPrepareTest _returnvalue))
  (Audio|Components|Audio|Stop (Variables|Default|GetAudio)))

(event EventStartTest
  (bind _stepsequencewidget (Variables|Default|GetStepSequenceWidget))
  (bind _generatorhandle (Variables|Default|GetGeneratorHandle))
  (Utilities|FlowControl|Sequence
    (:then_0
      (Development|PrintString "Step Sequence Player Test")
      (bind _returnvalue (MetaSound|CreateMetaSoundGeneratorHandle (Variables|Default|GetAudio)))
      (Variables|Default|SetGeneratorHandle _returnvalue)
      (bind _allsucceeded (Class|WBPTestStepSequence|AddWatchers _stepsequencewidget _generatorhandle (Variables|Default|GetOutputs)))
      (bind _success (Class|WBPTestStepSequence|AddNoteWatchers _stepsequencewidget _generatorhandle "Note On" "Note Off"))
      (if (not (and _success _allsucceeded))
        (FunctionalTesting|FinishTest "Failed" "Failed to watch outputs")))
    (:then_1
      (if (FunctionalTesting|IsRunning)
        (Development|PrintString "Step 0")
        (Audio|Components|Audio|Play (Variables|Default|GetAudio) 0.0)
        (Utilities|Time|SetTimerbyEvent (delegate Custom|Step%201) 4.0)))))

(event Custom|Step%201
  (Development|PrintString "Step 1")
  (Audio|Parameter|SetFloatParameter (Variables|Default|GetAudio) "Step Size Quarter Notes" 0.25 :!class /Script/AudioExtensions.AudioParameterControllerInterface)
  (Utilities|Time|SetTimerbyEvent (delegate Custom|Step%202) 4.0))

(event Custom|Step%202
  (Development|PrintString "Step 2")
  (Audio|Parameter|SetFloatParameter (Variables|Default|GetAudio) "Speed" 1.25 :!class /Script/AudioExtensions.AudioParameterControllerInterface)
  (Utilities|FlowControl|Delay 1.0)
  (Audio|Parameter|SetFloatParameter (Variables|Default|GetAudio) "Speed" 2.0 :!class /Script/AudioExtensions.AudioParameterControllerInterface)
  (Utilities|FlowControl|Delay 1.0)
  (Audio|Parameter|SetFloatParameter (Variables|Default|GetAudio) "Speed" 0.1 :!class /Script/AudioExtensions.AudioParameterControllerInterface)
  (Utilities|FlowControl|Delay 1.0)
  (Audio|Parameter|SetFloatParameter (Variables|Default|GetAudio) "Speed" 1.0 :!class /Script/AudioExtensions.AudioParameterControllerInterface)
  (Utilities|Time|SetTimerbyEvent (delegate Custom|Step%203) 4.0))

(event Custom|Step%203
  (Development|PrintString "Step 3")
  (Audio|Parameter|SetFloatParameter (Variables|Default|GetAudio) "Additional Octaves" 1.0 :!class /Script/AudioExtensions.AudioParameterControllerInterface)
  (Utilities|FlowControl|Delay 4.0)
  (Audio|Parameter|SetFloatParameter (Variables|Default|GetAudio) "Additional Octaves" 2.0 :!class /Script/AudioExtensions.AudioParameterControllerInterface)
  (Utilities|FlowControl|Delay 4.0)
  (Audio|Parameter|SetFloatParameter (Variables|Default|GetAudio) "Additional Octaves" -2.0 :!class /Script/AudioExtensions.AudioParameterControllerInterface)
  (Utilities|FlowControl|Delay 4.0)
  (Audio|Parameter|SetFloatParameter (Variables|Default|GetAudio) "Additional Octaves" 0.0 :!class /Script/AudioExtensions.AudioParameterControllerInterface)
  (Utilities|Time|SetTimerbyEvent (delegate Custom|Step%204) 4.0))

(event Custom|Step%204
  (Development|PrintString "Step 4")
  (Audio|Parameter|SetFloatParameter (Variables|Default|GetAudio) "Step Size Quarter Notes" 0.125 :!class /Script/AudioExtensions.AudioParameterControllerInterface)
  (Utilities|FlowControl|Delay 4.0)
  (Audio|Parameter|SetFloatParameter (Variables|Default|GetAudio) "Step Size Quarter Notes" 2.0 :!class /Script/AudioExtensions.AudioParameterControllerInterface)
  (Utilities|FlowControl|Delay 4.0)
  (Audio|Parameter|SetFloatParameter (Variables|Default|GetAudio) "Step Size Quarter Notes" 0.5 :!class /Script/AudioExtensions.AudioParameterControllerInterface)
  (Utilities|FlowControl|Delay 4.0)
  (Audio|Parameter|SetFloatParameter (Variables|Default|GetAudio) "Step Size Quarter Notes" 0.5 :!class /Script/AudioExtensions.AudioParameterControllerInterface)
  (Utilities|Time|SetTimerbyEvent (delegate Custom|Step%205) 4.0))

(event Custom|Step%205
  (Development|PrintString "Step 5")
  (Audio|Components|Audio|Stop (Variables|Default|GetAudio))
  (FunctionalTesting|FinishTest "Succeeded"))

(fn ConstructionScript)
