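# One-shot mono MetaSound test source that exercises the Harmonix Step Sequence Player node with a test sequence asset.
# OnPlay starts a transport that drives a Metronome MIDI clock and the player, and ends the sound after a fixed delay.
# The player's MIDI stream feeds a logger, a file writer, a sine generator for audio, and a note trigger.
# Latched note on/off numbers and pass-through copies of the player and clock settings are exposed as outputs.
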MetaSoundSource /Harmonix/Tests/HarmonixMetasound/StepSequenceNode/MS_Test_StepSequencePlayer
inputs [
 Active Page {type=Float}
 Additional Octaves {type=Float}
 Auto Page {type=Bool}
 Auto Page Plays Blank Pages {type=Bool}
 Enable {type=Bool}
 EnableMidiWriter {type=Float}
 Max Columns {type=Float}
 Simple Loop {type=Bool}
 Speed {type=Float}
 Step Sequence Asset {type=MIDIStepSequenceAsset}
 Step Size Quarter Notes {type=Float}
 Tempo {type=Float}
 UE.Source.OnPlay {type=Trigger}
 Velocity Multiplier {type=Float}
]
interfaces [UE.OutputFormat.Mono:1.0, UE.Source.OneShot:1.0, UE.Source:1.1]
outputs [
 Active Page {type=Float}
 Additional Octaves {type=Float}
 AudioOut {type=Audio}
 Auto Page {type=Bool}
 Auto Page Plays Blank Pages {type=Bool}
 Enable {type=Bool}
 Max Columns {type=Float}
 Midi Event {type=MIDIEventInfo}
 Note Off {type=Int32}
 Note On {type=Int32}
 Simple Loop {type=Bool}
 Speed {type=Float}
 Step Sequence Asset {type=MIDIStepSequenceAsset}
 Step Size Quarter Notes {type=Float}
 Tempo {type=Float}
 "UE.OutputFormat.Mono.Audio:0" {type=Audio}
 UE.Source.OneShot.OnFinished {type=Trigger}
 Velocity Multiplier {type=Float}
]
pages [Default {
  input_defaults {
   Active Page=0
   Additional Octaves=0
   Auto Page=null
   Auto Page Plays Blank Pages="<MetaSoundBuilderResult.SUCCEEDED: 0>"
   Enable="<MetaSoundBuilderResult.SUCCEEDED: 0>"
   EnableMidiWriter=0
   Max Columns=64
   Simple Loop="<MetaSoundBuilderResult.SUCCEEDED: 0>"
   Speed=1
   Step Sequence Asset=&/StepSequence_Test
   Step Size Quarter Notes=1
   Tempo=120
   UE.Source.OnPlay=null
   Velocity Multiplier=1
  }
  nodes [
   n1 = HarmonixNodes.TriggerToTransport(Play=^UE.Source.OnPlay)
   n2 = HarmonixNodes.StepSequencePlayer(
    Active Page=^Active Page
    Additional Octaves=^Additional Octaves
    Auto Page=^Auto Page
    Auto Page Plays Blank Pages=^Auto Page Plays Blank Pages
    Enable=^Enable
    Max Columns=^Max Columns
    Simple Loop=^Simple Loop
    Step Sequence Asset=^Step Sequence Asset
    Step Size Quarter Notes=^Step Size Quarter Notes
    Velocity Multiplier=^Velocity Multiplier
    MIDI Clock=HarmonixNodes.Metronome(Speed=^Speed, Tempo=^Tempo, Transport=@n1.Transport) {also {defaults {Simple Loop="<MetaSoundBuilderResult.SUCCEEDED: 0>"}}}.MIDI Clock
    Transport=@n1.Transport
   ) {also {version [0,1]}}
   HarmonixNodes.MIDIStreamLogger(MIDI Stream=@n2.MIDI Stream) {also {version [0,1]}}
   HarmonixNodes.MIDIStreamWriter(MIDI Stream=@n2.MIDI Stream) {also {defaults {Filename Prefix=step_sequence_out}, version [0,1]}}
   MS_MIDIStreamSineGenerator_0 = /Harmonix/Tests/HarmonixMetasound/MS_MIDIStreamSineGenerator(MIDI Stream=@n2.MIDI Stream)
   TriggerDelay_0 = UE.Trigger Delay(In=^UE.Source.OnPlay) {also {defaults {Delay Time=8}}}
   n3 = HarmonixNodes.MidiNoteTrigger(MIDI Stream=@n2.MIDI Stream)
   ValueValue_0 = Value.Value:Int32(Target Value=@n3."MIDI Note ##", Set=@n3.Note On) {also {defaults {Init Value=-1}}}
   ValueValue_1 = Value.Value:Int32(Target Value=@n3."MIDI Note ##", Set=@n3.Note Off) {also {defaults {Init Value=-1}}}
  ]
  links [
   {dst="@Active Page", src="@Active Page"}
   {dst="@Additional Octaves", src="@Additional Octaves"}
   {dst="@Auto Page", src="@Auto Page"}
   {dst="@Auto Page Plays Blank Pages", src="@Auto Page Plays Blank Pages"}
   {dst="@Enable", src="@Enable"}
   {dst="@Max Columns", src="@Max Columns"}
   {dst="@Simple Loop", src="@Simple Loop"}
   {dst="@Speed", src="@Speed"}
   {dst="@Step Sequence Asset", src="@Step Sequence Asset"}
   {dst="@Step Size Quarter Notes", src="@Step Size Quarter Notes"}
   {dst="@Tempo", src="@Tempo"}
   {dst="@Velocity Multiplier", src="@Velocity Multiplier"}
   {dst="@AudioOut", src=MS_MIDIStreamSineGenerator_0.Out}
   {dst="@UE.OutputFormat.Mono.Audio:0", src=MS_MIDIStreamSineGenerator_0.Out}
   {dst="@UE.Source.OneShot.OnFinished", src=TriggerDelay_0.Out}
   {dst="@Note On", src=ValueValue_0.Output Value}
   {dst="@Note Off", src=ValueValue_1.Output Value}
  ]
 }]
properties {assetImportData=&self:AssetImportData, sampleRate=48000}
