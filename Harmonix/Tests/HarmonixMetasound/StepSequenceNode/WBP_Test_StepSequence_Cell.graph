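# Test widget for the Harmonix MetaSound step sequence node that shows one step cell as a fixed-size square with a centered note number.
# SetNoteNumber stores the number and writes it to the text label, and GetNoteNumber returns it.
# SetNoteOn stores the on state and switches the border color between green (on) and gray (off).
# PreConstruct starts the cell in the off state, and the cell has no input handling of its own.
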WidgetBlueprint /Harmonix/Tests/HarmonixMetasound/StepSequenceNode/WBP_Test_StepSequence_Cell
graphs [
 EventGraph {nodes [
   Event PreConstruct()
   > SetNoteOn!(execute, self, IsOn:bool=false) -> (then)
   Event Construct()
   Event Tick()
  ]}
 GetNoteNumber {nodes [
   GetNoteNumber() {is=title K2Node_FunctionEntry}
   > Return Node!(execute, Return Value:int=Get NoteNumber() -> (NoteNumber:int=0))
  ]}
 SetNoteNumber {nodes [
   n1 = SetNoteNumber!() -> (then, NoteNumber:int) {is=title K2Node_FunctionEntry}
   > n2 = Set NoteNumber(NoteNumber:int=@n1) -> (Output_Get:int=0)
   > SetText!(execute, self:Self Object Reference, NewParam:Text=Format Text!(Format="INVTEXT(\"{Note ##}\")", "Note ##":int=@n2) -> (Result)) -> (then)
  ]}
 SetNoteOn {nodes [
   n1 = SetNoteOn!() -> (then, IsOn:bool=~) {is=title K2Node_FunctionEntry}
   > n2 = Set NoteOn(NoteOn:bool=@n1) -> (Output_Get:bool)
   > SetBrushColor(
    self=Get Border_17() -> (Border_17:Border Object Reference)
    InBrushColor=Select(Option 0:Linear Color Structure=(R=0.2188,G=0.206,B=0.1979,A=1), Option 1:Linear Color Structure=(R=0,G=1,B=0.0976,A=1), Index=@n2) -> (ReturnValue:Linear Color Structure)
   )
  ]}
 SetText {nodes [
   n1 = SetText!() -> (then, NewParam:Text) {is=title K2Node_FunctionEntry}
   > Set Text(Text=@n1, self=Get Text!() -> (Text:Text Object Reference))
  ]}
]
variables [
 NoteNumber {type=integer}
 NoteOn {type=boolean}
]
widgets {
 info {inheritedWidgetCount=0, parentClass=&UserWidget, rootWidgetClass=&SizeBox, widgetCount=3}
 widgets [
  SizeBox_36 {
   widgetClassPath=&SizeBox
   widget=SizeBox_36
   properties {
    bOverride_MaxDesiredHeight=true
    bOverride_MaxDesiredWidth=true
    bOverride_MinDesiredHeight=true
    bOverride_MinDesiredWidth=true
    maxDesiredHeight=50
    maxDesiredWidth=50
    minDesiredHeight=50
    minDesiredWidth=50
   }
  }
  Border_17 {
   widgetClassPath=&Border
   bIsVariable=true
   parent=SizeBox_36
   widget=Border_17
   properties {horizontalAlignment=HAlign_Center, verticalAlignment=VAlign_Center}
   slot {class=SizeBoxSlot}
  }
  Text {
   widgetClassPath=&TextBlock
   bIsVariable=true
   parent=Border_17
   widget=Text
   properties {
    clipping=ClipToBounds
    text=-100
    colorAndOpacity {colorUseRule=UseColor_Specified, specifiedColor=<rgba 0 0 0 1>}
    font {
     bForceMonospaced=false
     bMaterialIsStencil=false
     fontMaterial=null
     fontObject=&Roboto
     letterSpacing=0
     monospacedWidth=1
     size=12
     skewAmount=0
     typefaceFontName=Bold
     outlineSettings {
      bApplyOutlineToDropShadows=false
      bMiteredCorners=false
      bSeparateFillAlpha=false
      outlineMaterial=null
      outlineSize=0
      outlineColor=<rgba 0 0 0 1>
     }
    }
   }
   slot {class=BorderSlot, properties {horizontalAlignment=HAlign_Center, verticalAlignment=VAlign_Center}}
  }
 ]
}
