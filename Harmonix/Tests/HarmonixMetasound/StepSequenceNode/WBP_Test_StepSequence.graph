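# Test widget for the Harmonix MetaSound step sequence node that visualises notes and generator outputs.
# PreConstruct builds a grid of 128 note cells, which SetNote switches on or off by note number from Note On/Off output callbacks.
# Watcher functions subscribe to MetaSound generator outputs, adding a text row for each watched name.
# Updates try typed getters in turn to format values, setting a highlight colour that Tick interpolates back to white.
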WidgetBlueprint /Harmonix/Tests/HarmonixMetasound/StepSequenceNode/WBP_Test_StepSequence
graphs [
 Add Note Watchers {nodes [
   n1 = Add Note Watchers!() -> (then, Generator Handle:Meta Sound Generator Handle Object Reference, Note On:Name, Note Off:Name) {is=title K2Node_FunctionEntry}
   > n2 = AddWatchers!(
    execute
    self
    Generator Handle:Meta Sound Generator Handle Object Reference=@n1
    Outputs:Array of Names=Make Array!("[0]":Name=@n1.Note On, "[1]":Name=@n1.Note Off) -> (Array:Array of Names)
   ) -> (then, AllSucceeded:bool)
   > n3 = Branch(Condition=@n2)
   > n4 = WatchOutput!(
    execute
    self:Meta Sound Generator Handle Object Reference=@n1
    OutputName:Name=@n1.Note On
    OnOutputValueChanged:"Delegate (by ref)"=Create Event()
    AnalyzerName:Name=None
    AnalyzerOutputName:Name=None
   ) -> (then, ReturnValue:bool)
   > n5 = Branch(Condition=@n4)
   > n6 = WatchOutput(self=@n1, OutputName=@n1.Note Off, OnOutputValueChanged=Create Event(), AnalyzerName=None, AnalyzerOutputName=None)
   > n7 = Branch(Condition=@n6)
   > Return Node!(execute, Success:bool=true)
   Return Node(execute=[@n7.else, @n5.else, @n3.else], Success=false)
  ]}
 AddWatchers {nodes [
   n1 = AddWatchers!() -> (then, Generator Handle:Meta Sound Generator Handle Object Reference, Outputs:Array of Names) {is=title K2Node_FunctionEntry}
   > Set Succeeded(Succeeded:bool=true) -> (Output_Get:bool)
   > n2 = For Each Loop(Array:Array of Names=@n1.Outputs) -> (Array Element:Name)
   Return Node!(execute=@n2.Completed, AllSucceeded:bool=Get Succeeded() -> (Succeeded:bool))
   > n3 = WatchOutput(self=@n1, OutputName=@n2, OnOutputValueChanged=Create Event(), AnalyzerName=None, AnalyzerOutputName=None)
   > Set Succeeded(Succeeded=AND Boolean(A=@n3, B=Get Succeeded()))
   > Branch(Condition=@n3)
   > n4 = CreateWatcherWidget!(execute, self, Output Name:Name=@n2) -> (then, ReturnValue:Widget Object Reference)
   > n5 = AddChild(self=Get VBox_Properties() -> (VBox_Properties:Vertical Box Object Reference), Content=@n4)
   > n6 = Cast To VerticalBoxSlot!(execute, Object=@n5) -> (then, CastFailed, AsVertical Box Slot:Vertical Box Slot Object Reference) {is=title K2Node_DynamicCast}
   > SetPadding!(execute, self:Vertical Box Slot Object Reference=@n6, InPadding:Margin Structure=Make Margin(Left=5, Top=5, Right=5, Bottom=5)) -> (then)
  ]}
 BuildNotes {nodes [
   BuildNotes() {is=title K2Node_FunctionEntry}
   > ClearChildren(self=Get GridPanel() -> (GridPanel:Uniform Grid Panel Object Reference))
   > n1 = For Loop(LastIndex=127)
   > n2 = Create Widget!(
    execute:Exec
    Class:User Widget Class Reference=&/WBP_Test_StepSequence_Cell.WBP_Test_StepSequence_Cell_C
    OwningPlayer:Player Controller Object Reference
   ) -> (then:Exec, ReturnValue:WBP Test Step Sequence Cell Object Reference)
   > SetNoteNumber!(execute, self:WBP Test Step Sequence Cell Object Reference=@n2, NoteNumber:int=@n1) -> (then)
   > SetNoteOn!(execute, self:WBP Test Step Sequence Cell Object Reference=@n2, IsOn:bool=false) -> (then)
   > n3 = AddChild(self=Get GridPanel(), Content=@n2)
   > n4 = Cast To UniformGridSlot!(execute, Object=@n3) -> (then, CastFailed, AsUniform Grid Slot:Uniform Grid Slot Object Reference) {is=title K2Node_DynamicCast}
   n5 = Get Size() -> (Size:int=0)
   > SetColumn!(execute, self:Uniform Grid Slot Object Reference=@n4, InColumn:int=int / int(A=@n1, B=@n5)) -> (then)
   > SetRow!(execute, self:Uniform Grid Slot Object Reference=@n4, InRow:int="% (Integer)"(A=@n1, B=@n5) -> (ReturnValue=0)) -> (then)
   SetPadding(InPadding=Make Margin(Left=10, Top=10, Right=10, Bottom=10))
  ]}
 CreateWatcherWidget {nodes [
   n1 = CreateWatcherWidget!() -> (then, Output Name:Name) {is=title K2Node_FunctionEntry}
   > n2 = Construct Object from Class(Class=&TextBlock) -> (ReturnValue:Text Object Reference)
   > Add!(
    execute
    TargetMap:Map of Names to Widget Object References=Get WatcherWidgets() -> (WatcherWidgets:Map of Names to Widget Object References)
    Key:"Name (by ref)"=@n1
    Value:"Widget Object Reference (by ref)"=@n2
   ) -> (then) {is=title K2Node_CallFunction}
   > Set Text(Text=Format Text!(Format="INVTEXT(\"{Name}: ##NONE\")", Name:Name=@n1) -> (Result), self=@n2)
   > SetFont(
    self=@n2
    InFontInfo=MakeSlateFontInfo!(
     FontObject:Object Reference=&Roboto
     FontMaterial:Object Reference
     OutlineSettings:Font Outline Settings Structure=(OutlineSize=0,bMiteredCorners=False,bSeparateFillAlpha=False,bApplyOutlineToDropShadows=False,OutlineMaterial=None,OutlineColor=(R=0,G=0,B=0,A=1))
     TypefaceFontName:Name=Regular
     Size:float=12
     LetterSpacing:int=0
     SkewAmount:float=0
     bForceMonospaced:bool=False
     bMaterialIsStencil:bool=False
     MonospacedWidth:float=1
    ) -> (ReturnValue:Slate Font Info Structure)
   )
   > Return Node!(execute, ReturnValue:Widget Object Reference=@n2)
  ]}
 EventGraph {nodes [
   Event PreConstruct()
   > BuildNotes() {is=title K2Node_CallFunction}
   n1 = Event Tick()
   > n2 = Values!(execute, TargetMap:Map of Names to Widget Object References=Get WatcherWidgets()) -> (then, Values:Array of Widget Object References)
   > n3 = For Each Loop(Array:Array of Widget Object References=@n2) -> (Array Element:Widget Object Reference)
   > n4 = Cast To TextBlock!(execute, Object=@n3) -> (then, CastFailed, AsText:Text Object Reference) {is=title K2Node_DynamicCast}
   n5 = Get!(self:Text Object Reference=@n4) -> (ColorAndOpacity_SpecifiedColor:Linear Color Structure, ColorAndOpacity_ColorUseRule:ESlateColorStylingMode Enum)
   > Set ColorAndOpacity!(
    execute
    ColorAndOpacity:Slate Color Structure=Make Slate Color(SpecifiedColor="Interpolate (LinearColor)"!(Current:Linear Color Structure=@n5, Target:Linear Color Structure=(R=1,G=1,B=1,A=1), DeltaTime:float=@n1.InDeltaTime, InterpSpeed:float=0.5) -> (ReturnValue:Linear Color Structure))
    self:Text Object Reference=@n4
   ) -> (then, Output_Get:Slate Color Structure)
   Event Construct()
  ]}
 HighlightAndFadeWidgetWatcher {nodes [
   n1 = HighlightAndFadeWidgetWatcher!() -> (then, Widget:"Widget Object Reference (by ref)") {is=title K2Node_FunctionEntry}
   > n2 = Cast To TextBlock(Object=@n1)
   > SetColorAndOpacity(self=@n2, InColorAndOpacity=Get HighlightColor() -> (HighlightColor:Slate Color Structure))
  ]}
 OnOutputWatcherNoteOffValueChanged {nodes [
   n1 = OnOutputWatcherNoteOffValueChanged!() -> (then, OutputName:Name, Output:"Meta Sound Output Structure (by ref)") {is=title K2Node_FunctionEntry}
   > n2 = GetInt32!(execute, Output:"Meta Sound Output Structure (by ref)"=@n1.Output) -> (then, Success:bool, ReturnValue:int=0)
   > Branch(Condition=@n2)
   > SetNote!(execute, self, "Note ##":int=@n2.ReturnValue, On:bool=false) -> (then)
   > Return Node!(execute)
  ]}
 OnOutputWatcherNoteOnValueChanged {nodes [
   n1 = OnOutputWatcherNoteOnValueChanged!() -> (then, OutputName:Name, Output:"Meta Sound Output Structure (by ref)") {is=title K2Node_FunctionEntry}
   > n2 = GetInt32(Output=@n1.Output) -> (ReturnValue=0)
   > Branch(Condition=@n2)
   > SetNote("Note ##"=@n2.ReturnValue, On=true)
   > Return Node()
  ]}
 OnOutputWatcherValueChanged {nodes [
   n1 = OnOutputWatcherValueChanged!() -> (then, OutputName:Name, Output:"Meta Sound Output Structure (by ref)") {is=title K2Node_FunctionEntry}
   > UpdateWatcherWidget!(execute, self, OutputName:Name=@n1, Output:Meta Sound Output Structure=@n1.Output) -> (then)
   > Return Node()
  ]}
 SetNote {nodes [
   SetNote!() -> (then, "Note ##":int, On:bool=~) {is=title K2Node_FunctionEntry}
   n1 = Get GridPanel()
   > n2 = For Loop(LastIndex=int - int(A=GetChildrenCount!(self:Panel Widget Object Reference=@n1) -> (ReturnValue:int=0)))
   > n3 = Cast To WBP_Test_StepSequence_Cell!(execute, Object=GetChildAt!(self:Panel Widget Object Reference=@n1, Index:int=@n2) -> (ReturnValue:Widget Object Reference)) -> (then, CastFailed, AsWBP Test Step Sequence Cell:WBP Test Step Sequence Cell Object Reference) {is=title K2Node_DynamicCast}
   > Branch(Condition="Equal (Integer)"(A="Get Note ##"() -> ("Note ##":int=0), B=GetNoteNumber!(self:WBP Test Step Sequence Cell Object Reference=@n3) -> (Return Value:int=0)))
   > SetNoteOn(self=@n3, IsOn=Get On() -> (On:bool))
  ]}
 UpdateWatcherWidget {nodes [
   n1 = UpdateWatcherWidget!() -> (then, OutputName:Name, Output:Meta Sound Output Structure) {is=title K2Node_FunctionEntry}
   n2 = Find(TargetMap:Map of Names to Widget Object References=Get WatcherWidgets(), Key:"Name (by ref)"=@n1) -> (Value:Widget Object Reference)
   > n3 = Set WatcherWidget(WatcherWidget:Widget Object Reference=@n2) -> (Output_Get:Widget Object Reference)
   > Branch(Condition=@n2.ReturnValue)
   > HighlightAndFadeWidgetWatcher!(execute, self, Widget:"Widget Object Reference (by ref)"=@n3) -> (then) {is=title K2Node_CallFunction}
   > n4 = GetInt32(Output=@n1.Output) -> (ReturnValue=0)
   > n5 = Branch(Condition=@n4)
   > n6 = Cast To TextBlock(Object=Get WatcherWidget() -> (WatcherWidget:Widget Object Reference))
   > Set Text(
    Text=Format Text!(
     Format="INVTEXT(\"{Output Name}: {Output Value}\")"
     Output Name:Name=Get OutputName() -> (OutputName:Name=None)
     Output Value:int=@n4.ReturnValue
    ) -> (Result)
    self=@n6
   )
   n7 = GetFloat!(execute=@n5.else, Output:"Meta Sound Output Structure (by ref)"=@n1.Output) -> (then, Success:bool, ReturnValue:float)
   > n8 = Branch(Condition=@n7)
   > n9 = Cast To TextBlock(Object=Get WatcherWidget())
   > Set Text(
    Text=Format Text(Format="INVTEXT(\"{Output Name}: {Output Value}\")", Output Name=Get OutputName() -> (OutputName=None), Output Value:float=@n7.ReturnValue)
    self=@n9
   )
   n10 = GetBool!(execute=@n8.else, Output:"Meta Sound Output Structure (by ref)"=@n1.Output) -> (then, Success:bool, ReturnValue:bool)
   > n11 = Branch(Condition=@n10)
   > n12 = Cast To TextBlock(Object=Get WatcherWidget())
   > Set Text(
    Text=Format Text(Format="INVTEXT(\"{Output Name}: {Output Value}\")", Output Name=Get OutputName() -> (OutputName=None), Output Value:bool=@n10.ReturnValue)
    self=@n12
   )
   n13 = GetString!(execute=@n11.else, Output:"Meta Sound Output Structure (by ref)"=@n1.Output) -> (then, Success:bool, ReturnValue:String)
   > n14 = Branch(Condition=@n13)
   > n15 = Cast To TextBlock(Object=Get WatcherWidget())
   > Set Text(
    Text=Format Text(Format="INVTEXT(\"{Output Name}: {Output Value}\")", Output Name=Get OutputName() -> (OutputName=None), Output Value:String=@n13.ReturnValue)
    self=@n15
   )
   n16 = GetMidiEventInfo!(execute=@n14.else, Output:"Meta Sound Output Structure (by ref)"=@n1.Output) -> (then, Success:bool, ReturnValue:MIDI Event Info Structure)
   > n17 = Branch(Condition=@n16)
   > n18 = Cast To TextBlock(Object=Get WatcherWidget())
   n19 = Break MIDI Event Info!(MidiEventInfo:"MIDI Event Info Structure (by ref)"=@n16.ReturnValue) -> (Timestamp_Bar:int, Timestamp_Beat:float=~, TrackIndex:int) {is=title K2Node_BreakStruct}
   > Set Text(
    Text=Format Text!(
     Format="INVTEXT(\"{Output Name}: {Track} - {Bar}:{Beat} \")"
     Output Name=Get OutputName() -> (OutputName=None)
     Track:int=@n19.TrackIndex
     Bar:int=@n19
     Beat:float=@n19.Timestamp_Beat
    ) -> (Result)
    self=@n18
   )
   n20 = GetMusicTimestamp!(execute=@n17.else, Output:"Meta Sound Output Structure (by ref)"=@n1.Output) -> (then, Success:bool, ReturnValue:Music Timestamp Structure)
   > n21 = Branch(Condition=@n20)
   > n22 = Cast To TextBlock(Object=Get WatcherWidget())
   n23 = Break Music Timestamp!(MusicTimestamp:"Music Timestamp Structure (by ref)"=@n20.ReturnValue) -> (Bar:int, Beat:float=~) {is=title K2Node_BreakStruct}
   > Set Text(
    Text=Format Text!(Format="INVTEXT(\"{Output Name}: {Bar}:{Beat}\")", Output Name=Get OutputName() -> (OutputName=None), Bar=@n23, Beat=@n23.Beat) -> (Result)
    self=@n22
   )
   n24 = GetTimeSeconds!(execute=@n21.else, Output:"Meta Sound Output Structure (by ref)"=@n1.Output) -> (then, Success:bool, ReturnValue:double)
   > n25 = Branch(Condition=@n24)
   > n26 = Cast To TextBlock(Object=Get WatcherWidget())
   > Set Text(
    Text=Format Text!(Format="INVTEXT(\"{Output Name}: {Output Value}\")", Output Name=Get OutputName() -> (OutputName=None), Output Value:double=@n24.ReturnValue) -> (Result)
    self=@n26
   )
   n27 = GetTimeSignature!(execute=@n25.else, Output:"Meta Sound Output Structure (by ref)"=@n1.Output) -> (then, Success:bool, ReturnValue:Time Signature Structure)
   > Branch(Condition=@n27)
   > n28 = Cast To TextBlock(Object=Get WatcherWidget())
   > Set Text(
    Text=Format Text!(Format="INVTEXT(\"{Output Name}: UNSUPPORTED\")", Output Name=Get OutputName() -> (OutputName=None)) -> (Result)
    self=@n28
   )
  ]}
]
variables [
 HighlightColor {type=SlateColor, default {colorUseRule=UseColor_Specified, specifiedColor=<rgba 0 0 0 1>}}
 WatcherWidgets {type="map<Widget>", default {}}
 WatcherWidgetUpdateStack {type="array<Widget>", default []}
]
widgets {
 info {inheritedWidgetCount=0, parentClass=&UserWidget, rootWidgetClass=&CanvasPanel, widgetCount=5}
 widgets [
  CanvasPanel_0 {widgetClassPath=&CanvasPanel, widget=CanvasPanel_0}
  GridPanel {
   widgetClassPath=/Script/UMG.UniformGridPanel
   bIsVariable=true
   parent=CanvasPanel_0
   widget=GridPanel
   properties {slotPadding=<ltrb 5 5 5 5>}
   slot {
    class=CanvasPanelSlot
    properties {bAutoSize=true, layoutData {alignment=<0.5 0.5>, anchors {maximum=<0.5 0.5>, minimum=<0.5 0.5>}, offsets=<ltrb 0 0 50 925.8687>}}
   }
  }
  Border_23 {
   widgetClassPath=&Border
   parent=CanvasPanel_0
   widget=Border_23
   properties {brushColor=<rgba 0 0 0 1>}
   slot {
    class=CanvasPanelSlot
    properties {layoutData {alignment=<0 0>, anchors {maximum=<0 1>, minimum=<0 0>}, offsets=<ltrb 58.5054 144 371.3062 156.1078>}}
   }
  }
  VBox_Properties {widgetClassPath=&VerticalBox, bIsVariable=true, parent=Border_23, widget=VBox_Properties, slot {class=BorderSlot}}
  TextBlock_Title {
   widgetClassPath=&TextBlock
   parent=VBox_Properties
   widget=TextBlock_Title
   properties {text=Step Sequence Player}
   slot {class=VerticalBoxSlot}
  }
 ]
}
